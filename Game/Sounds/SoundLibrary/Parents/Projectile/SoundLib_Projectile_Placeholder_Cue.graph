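# Placeholder SoundCue for projectile sounds, stored in the Parents/Projectile folder of the sound library.
# It has no root node and an empty node list, so it references no audio and plays nothing.
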SoundCue /Game/Sounds/SoundLibrary/Parents/Projectile/SoundLib_Projectile_Placeholder_Cue
root=null
nodes []
